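; Weapon Blueprint for the BentBaton cosmetic pickaxe that feeds swing state to its animation trail VFX.
; On each left or right swing it calls the parent event, then stores the swing side in IsRight.
; It then sets two bool parameters on the trail VFX component: whether the pickaxe is displayed in a frontend context and whether the swing is to the right.
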(event EventSwingLeft
  (bind _returnvalue (CallFunction|GetAnimTrailVFXComponent))
  (|CallParentFunction)
  (Variables|Default|SetIsRight false)
  (Effects|Components|ParticleSystem|SetBoolParameter _returnvalue "bIsFrontEndPreview" (Reactive|Frontend|IsCosmeticObjectDisplayedinFrontendContext 0))
  (Effects|Components|ParticleSystem|SetBoolParameter _returnvalue "Is RightSwing" (Variables|Default|GetIsRight)))

(event EventSwingRight
  (bind _returnvalue (CallFunction|GetAnimTrailVFXComponent))
  (|CallParentFunction)
  (Variables|Default|SetIsRight true)
  (Effects|Components|ParticleSystem|SetBoolParameter _returnvalue "bIsFrontEndPreview" (Reactive|Frontend|IsCosmeticObjectDisplayedinFrontendContext 0))
  (Effects|Components|ParticleSystem|SetBoolParameter _returnvalue "Is RightSwing" (Variables|Default|GetIsRight)))

(fn ConstructionScript
  (|CallParentFunction))
