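# Translucent, unlit, two-sided Niagara mesh-particle material for the glowing teeth effect on the BentBaton pickaxe.
# Emissive layers tinted grayscale teeth textures with blurred noise and particle color.
# Opacity combines an alpha texture, particle alpha, U-edge mask, view-angle and depth fades.
# Ceil-stepped time blended with smooth time gives choppy panning, while dual sine waves drive world position offset, with dynamic parameters controlling U offset, displacement and side fade.
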Material /BRCosmetics/Weapons/FORT_Melee/Pickaxe_BentBaton/FX/Materials/M_Pickaxe_BentBaton_Teeth
graph {
 properties {
  bAutomaticallySetUsageInEditor=false
  bUsedWithNiagaraMeshParticles=true
  blendMode=BLEND_Translucent
  shadingModel=MSM_Unlit
  twoSided=true
  usedShadingModels=Unlit
 }
 nodes [
  n1 = ScalarParameter() {defaultValue=10, parameterName="WidthBase (S)"}
  n2 = StaticBool()
  n3 = ComponentMask(None=TextureCoordinate()) {r=true}
  n4 = ScalarParameter() {defaultValue=15, parameterName=WPO_UV_A}
  n5 = Time()
  n6 = ScalarParameter() {defaultValue=-2.5, parameterName=WPOSpeed_UV_A}
  n7 = Sine(None=Add(A=Multiply(A=@n3, B=@n4) {constB=0.5}, B=Multiply(A=@n5, B=@n6) {constB=-1}))
  n8 = Power(Base=Abs(None=@n7), Exp=ScalarParameter() {defaultValue=1, parameterName=WPOEdgePower})
  n9 = LinearInterpolate(Alpha=LinearInterpolate(A=Multiply(A=@n8) {constB=-1}, B=@n8, Alpha=Ceil(None=@n7))) {constA=-1}
  n10 = ComponentMask(None=TextureCoordinate()) {r=true}
  n11 = ScalarParameter() {defaultValue=4, parameterName=WPO_UV_B}
  n12 = OneMinus(None=Time())
  n13 = ScalarParameter() {defaultValue=2, parameterName=WPOSpeed_UV_B}
  n14 = Sine(None=Add(A=Multiply(A=@n10, B=@n11) {constB=0.5}, B=Multiply(A=@n12, B=@n13) {constB=5}))
  n15 = Power(Base=Abs(None=@n14), Exp=ScalarParameter() {defaultValue=1, parameterName=WPOEdgePower})
  n16 = Multiply(A=@n15) {constB=-1}
  n17 = Ceil(None=@n14)
  n18 = Multiply(
   A=LinearInterpolate(Alpha=LinearInterpolate(A=@n16, B=@n15, Alpha=@n17)) {constA=-1}
   B=ScalarParameter() {defaultValue=1, parameterName=MovementIntencity}
  )
  n19 = Constant() {r=-1}
  n20 = Multiply(A=Add(A=Multiply(A=@n9), B=Multiply(A=@n18, B=@n19)), B=VectorParameter() {parameterName=WPODirection, defaultValue=<rgba 0 0.4 0 0>}.RGB)
  n21 = DynamicParameter() {paramNames [U_Offset, WPO, Param3, UseWPOSideFade]}
  n22 = TextureCoordinate()
  n23 = Power(Base=Sine(None=ComponentMask(None=@n22) {r=true}) {period=2}) {constExponent=1.5}
  n24 = LinearInterpolate(B=Saturate(None=Multiply(A=@n23) {constB=5}), Alpha=@n21.UseWPOSideFade) {constA=1}
  Add(
   A=LinearInterpolate(
    B=MaterialFunctionCall(WidthBase=@n1, WidthTip=@n1, Expand U or V UV Channel=@n2) {materialFunction=&SplineThicken}.WorldPositionOffset
    Alpha=ScalarParameter() {parameterName=UseSplineTickenSwitch}
   )
   B=Multiply(A=Transform(None=@n20), B=Multiply(A=@n21.WPO, B=@n24))
  )
  n25 = ParticleColor()
  n26 = TextureSampleParameter2D(UVs=NamedRerouteUsage()) {
   parameterName=AlphaTex
   samplerType=SAMPLERTYPE_Grayscale
   texture=&/T_BentBaton_Grayscale_Teeth05
  }
  n27 = Multiply(
   A=@n25.A
   B=Power(
    Base=StaticComponentMaskParameter(None=@n26.RGB) {defaultG=true, parameterName=AlphaTexMask}
    Exp=ScalarParameter() {defaultValue=1, parameterName=Exp}
   ) {constExponent=1}
  )
  n28 = Sine(None=ComponentMask(None=TextureCoordinate()) {r=true}) {period=2}
  n29 = ScalarParameter() {defaultValue=2, parameterName=LinearMask_Power}
  n30 = Saturate(None=Multiply(A=Power(Base=@n28, Exp=@n29), B=ScalarParameter() {defaultValue=7, parameterName=LinearMask_Mul}) {constB=7})
  n31 = MaterialFunctionCall("FadeSensitivity (optional)"=ScalarParameter() {parameterName=AngleFade}) {materialFunction=&FadeBasedOnViewAngle}
  n32 = ScalarParameter() {parameterName="Contrast (S)"}
  DepthFade(
   Opacity=Multiply(
    A=Multiply(A=@n27, B=@n30)
    B=MaterialFunctionCall(In=@n31.Result, Contrast=@n32) {materialFunction=&CheapContrast}.Result
   )
   FadeDistance=ScalarParameter() {parameterName=FadeDistance}
  )
  n33 = VectorParameter() {parameterName=Color03, defaultValue=<rgba 0.9323 0.0481 0.916 1>}
  n34 = TextureSampleParameter2D(UVs=NamedRerouteUsage()) {
   parameterName=TexLight
   samplerType=SAMPLERTYPE_Grayscale
   texture=&/T_BentBaton_Grayscale_Teeth03
  }
  n35 = Multiply(A=MaterialFunctionCall(In=Power(Base=@n34.R) {constExponent=1}, Contrast=Constant() {r=0.5}) {materialFunction=&CheapContrast}.Result)
  n36 = TextureSample(UVs=Panner(Coordinate=TextureCoordinate() {uTiling=0.5}, Time=NamedRerouteUsage()) {speedX=-0.5}) {samplerType=SAMPLERTYPE_Grayscale, texture=&LowResBlurredNoise}
  n37 = VectorParameter() {parameterName=LerpTexColor, defaultValue=<rgba 0.7286 0.1406 1 1>}
  n38 = TextureSampleParameter2D(UVs=NamedRerouteUsage()) {
   parameterName=Lerp_TexGlow
   samplerType=SAMPLERTYPE_Grayscale
   texture=&/T_BentBaton_Grayscale_Teeth02
  }
  n39 = StaticComponentMaskParameter(None=@n38.RGB) {defaultR=true, parameterName=LerpTexMask}
  n40 = TextureSample(UVs=Panner(Coordinate=TextureCoordinate() {uTiling=0.5}, Time=NamedRerouteUsage()) {speedX=-0.5}) {samplerType=SAMPLERTYPE_Grayscale, texture=&LowResBlurredNoise}
  n41 = Add(
   A=Multiply(A=Multiply(A=@n33.RGB, B=@n35), B=LinearInterpolate(Alpha=@n36.R) {constB=3})
   B=Multiply(A=Multiply(A=@n37.RGB, B=@n39), B=LinearInterpolate(Alpha=@n40.R))
  )
  n42 = Panner(Coordinate=TextureCoordinate() {uTiling=0.5}, Time=NamedRerouteUsage()) {speedX=-0.5}
  n43 = Multiply(
   A=LinearInterpolate(
    A=ScalarParameter() {defaultValue=-1, parameterName=DarkNoiseLerpMin}
    Alpha=TextureSample(UVs=@n42) {samplerType=SAMPLERTYPE_Grayscale, texture=&LowResBlurredNoise}.R
   ) {constA=-1, constB=5}
   B=VectorParameter() {parameterName=DarkBackGroundColor, defaultValue=<rgba 0.0417 0 0.05 1>}.RGB
  )
  n44 = TextureSampleParameter2D(UVs=NamedRerouteUsage()) {
   parameterName=TexCenter
   samplerType=SAMPLERTYPE_Grayscale
   texture=&/T_BentBaton_Grayscale_Teeth01
  }
  n45 = Power(Base=@n44.R, Exp=ScalarParameter() {defaultValue=10, parameterName=CenterTexPower}) {constExponent=10}
  n46 = TextureCoordinate() {uTiling=2}
  n47 = Multiply(
   A=LinearInterpolate(
    A=Add(A=@n41, B=@n43)
    B=VectorParameter() {parameterName=CenterTexColor, defaultValue=<rgba 1 0.6479 0.8917 1>}.RGB
    Alpha=Saturate(None=@n45)
   )
   B=LinearInterpolate(Alpha=TextureSample(UVs=@n46) {samplerType=SAMPLERTYPE_Grayscale, texture=&LowResBlurredNoise}.R) {constA=0.8}
  )
  MaterialFunctionCall(Emissive=LinearInterpolate(A=Multiply(A=@n47, B=@n25.RGB), B=@n25.RGB, Alpha=ScalarParameter() {parameterName=UseSimpleGlow})) {materialFunction=&MF_FX_EmissiveAdjustments}
  n48 = Time()
  n49 = ScalarParameter() {defaultValue=1, parameterName=Speed}
  n50 = Multiply(A=Constant() {r=2.5}, B=@n49)
  n51 = Constant() {r=8}
  n52 = Ceil(None=Multiply(A=Time() {bOverride_Period=true, period=6000}, B=Multiply(A=@n51, B=@n49)) {constB=8})
  NamedRerouteDeclaration(None=LinearInterpolate(A=Multiply(A=@n48, B=@n50) {constB=2.3}, B=Multiply(A=@n52) {constB=0.5}, Alpha=ScalarParameter() {defaultValue=0.4, parameterName=LerpStep}) {constAlpha=1}) {name=StepPanner, nodeColor=<rgba 1 0.0368 0.001 1>}
  n53 = Multiply(
   A=TextureCoordinate()
   B=ComponentMask(None=VectorParameter() {parameterName=TexA_UVScale, defaultValue=<rgba 0.7 1 0 0>}.RGB) {g=true, r=true}
  )
  n54 = AppendVector(A=DynamicParameter() {paramNames [U_Offset, WPO, Param3, UseWPOSideFade]}.U_Offset, B=Constant())
  n55 = Time()
  n56 = Time() {bIgnorePause=true, bOverride_Period=true, period=6000}
  n57 = Multiply(A=Ceil(None=Multiply(A=@n56) {constB=10})) {constB=0.5}
  n58 = VectorParameter() {parameterName=StepPannerSpeed, defaultValue=<rgba 0.1 0 0 0>}
  NamedRerouteDeclaration(None=Panner(
    Coordinate=Add(A=@n53, B=@n54)
    Time=LinearInterpolate(A=@n55, B=@n57) {constAlpha=0.15}
    Speed=ComponentMask(None=@n58.RGB) {g=true, r=true}
   ) {speedX=0.5}) {name=TimeStep_L, nodeColor=<rgba 0.4627 0.0027 1 1>}
 ]
}
