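# Material function in the BentBaton pickaxe folder that tiles a texture in world space by triplanar projection, with per-axis rotation and depth offset.
# World position over texture size gives three planar UV sets, each rotated, then shifted by parallax occlusion mapping or bump offset via static switch.
# Contrast-adjusted world normal blends the samples, outputting full, side-only and Z-only results.
# Sample derivatives come from mesh UVs, not projected ones.
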MaterialFunction /BRCosmetics/Weapons/FORT_Melee/Pickaxe_BentBaton/Materials/MF_WorldAlignedTexture_RotBumpOffset
graph {
 properties {bExposeToLibrary=true, description=Tiles a texture in worldspace, libraryCategoriesText [Misc,Texturing]}
 nodes [
  n1 = FunctionInput(Preview=TextureObject() {texture=&/Swirly_Wisps_Vertical_Dissolve}) {bUsePreviewValueAsDefault=true, inputName=Heightmap, inputType=FunctionInput_Texture2D, previewValue=<1 1 1 1>}
  n2 = FunctionInput() {bUsePreviewValueAsDefault=true, inputName=Height Amount, inputType=FunctionInput_Scalar, previewValue=<1 0 0 1>}
  n3 = FunctionInput() {bUsePreviewValueAsDefault=true, inputName=Min Steps, inputType=FunctionInput_Scalar, previewValue=<8 0 0 1>}
  n4 = FunctionInput() {bUsePreviewValueAsDefault=true, inputName=Max Steps, inputType=FunctionInput_Scalar, previewValue=<32 0 0 1>}
  n5 = FunctionInput() {bUsePreviewValueAsDefault=true, description=Size of texture in worldspace units, inputName=TextureSize, previewValue=<64 64 64 6>}
  n6 = Divide(
   A=FunctionInput(Preview=WorldPosition() {worldPositionShaderOffset=WPT_ExcludeAllShaderOffsets}.XYZ) {bUsePreviewValueAsDefault=true, inputName=WorldPosition}
   B=Multiply(A=Abs(None=@n5)) {constB=-1}
  )
  n7 = MaterialFunctionCall(
   UVs=ComponentMask(None=@n6) {b=true, r=true}
   "Rotation Angle (0-1)"=FunctionInput() {bUsePreviewValueAsDefault=true, inputName=RotYAxis, inputType=FunctionInput_Scalar}
  ) {materialFunction=&CustomRotator}
  n8 = Constant4Vector() {constant=<rgba 1 0 0 1>}
  n9 = FunctionInput() {bUsePreviewValueAsDefault=true, inputName=Ref Plane, inputType=FunctionInput_Scalar, previewValue=<1 1 1 1>}
  n10 = StaticBoolParameter() {defaultValue=true, parameterName="Use World Coordinates (SB)"}
  n11 = FunctionInput() {bUsePreviewValueAsDefault=true, inputName=Parallax, inputType=FunctionInput_Scalar}
  n12 = FunctionInput() {bUsePreviewValueAsDefault=true, inputName=Bump Amount, inputType=FunctionInput_Scalar}
  n13 = FunctionInput(Preview=TextureObject() {texture=/Engine/EngineMaterials/DefaultDiffuse}) {inputName=TextureObject, inputType=FunctionInput_Texture2D}
  n14 = TextureCoordinate()
  n15 = DDX(Value=@n14)
  n16 = DDY(Value=@n14)
  n17 = TextureSample!(
   UVs=StaticSwitchParameter(
    True=MaterialFunctionCall!(
     Heightmap Texture=@n1
     Height Ratio=@n2
     Min Steps=@n3
     Max Steps=@n4
     UVs=@n7.Rotated Values
     Heightmap Channel=@n8.RGBA
     Reference Plane=@n9
     -----------------------------
     Use World Coordinates=@n10
     Specify Manual Texture Size
     Manual Texture Size
     ------------------------------
     "Render Shadows (Occlusion Mapping)"
     Light Vector
     Shadow Steps
     Shadow Penumbra
     Transform To VertexNormal
    ) {materialFunction=/Engine/Functions/Engine_MaterialFunctions01/Texturing/ParallaxOcclusionMapping}.Parallax UVs
    False=BumpOffset(Coordinate=@n7.Rotated Values, Height=@n11, HeightRatioInput=@n12)
   ) {parameterName=POM or Bump Offset?}
   Tex=@n13
   "DDX(UVs)"=@n15
   "DDY(UVs)"=@n16
   Apply View MipBias
  ) {mipValueMode=TMVM_Derivative, samplerSource=SSM_Wrap_WorldGroupSettings}
  n18 = FunctionInput(Preview=StaticBool()) {bUsePreviewValueAsDefault=true, description="Default = False", inputName=Export Float 4, inputType=FunctionInput_StaticBool}
  n19 = MaterialFunctionCall(
   UVs=ComponentMask(None=@n6) {b=true, g=true}
   "Rotation Angle (0-1)"=FunctionInput() {bUsePreviewValueAsDefault=true, inputName=RotXAxis, inputType=FunctionInput_Scalar}
  ) {materialFunction=&CustomRotator}
  n20 = TextureSample(
   UVs=StaticSwitchParameter(
    True=MaterialFunctionCall(
     Heightmap Texture=@n1
     Height Ratio=@n2
     Min Steps=@n3
     Max Steps=@n4
     UVs=@n19.Rotated Values
     Heightmap Channel=@n8.RGBA
     Reference Plane=@n9
     Use World Coordinates=@n10
    ) {materialFunction=&ParallaxOcclusionMapping}.Parallax UVs
    False=BumpOffset(Coordinate=@n19.Rotated Values, Height=@n11, HeightRatioInput=@n12)
   ) {parameterName=POM or Bump Offset?}
   Tex=@n13
   "DDX(UVs)"=@n15
   "DDY(UVs)"=@n16
  ) {mipValueMode=TMVM_Derivative, samplerSource=SSM_Wrap_WorldGroupSettings}
  n21 = FunctionInput(Preview=VertexNormalWS()) {bUsePreviewValueAsDefault=true, inputName=World Space Normal, previewValue=<0 0 1 1>}
  n22 = Abs(None=ComponentMask(None=@n21) {r=true})
  n23 = FunctionInput() {
   bUsePreviewValueAsDefault=true
   inputName=ProjectionTransitionContrast
   inputType=FunctionInput_Scalar
   sortPriority=10
   previewValue=<1 0 0 1>
  }
  n24 = LinearInterpolate(
   A=StaticSwitch(True=AppendVector(A=@n17.RGB, B=@n17.RGB), False=@n17.RGB, Value=@n18)
   B=StaticSwitch(True=AppendVector(A=@n20.RGB, B=@n20.RGB), False=@n20.RGB, Value=@n18)
   Alpha=ComponentMask(None=MaterialFunctionCall(In=@n22, Contrast=@n23) {materialFunction=&CheapContrast}.Result) {r=true}
  ) {desc=Float3}
  n25 = MaterialFunctionCall(
   UVs=ComponentMask(None=@n6) {g=true, r=true}
   "Rotation Angle (0-1)"=FunctionInput() {bUsePreviewValueAsDefault=true, inputName=RotZAxis, inputType=FunctionInput_Scalar}
  ) {materialFunction=&CustomRotator}
  n26 = TextureSample(
   UVs=StaticSwitchParameter(
    True=MaterialFunctionCall(
     Heightmap Texture=@n1
     Height Ratio=@n2
     Min Steps=@n3
     Max Steps=@n4
     UVs=@n25.Rotated Values
     Heightmap Channel=@n8.RGBA
     Reference Plane=@n9
     Use World Coordinates=@n10
    ) {materialFunction=&ParallaxOcclusionMapping}.Parallax UVs
    False=BumpOffset(Coordinate=@n25.Rotated Values, Height=@n11, HeightRatioInput=@n12)
   ) {parameterName=POM or Bump Offset?}
   Tex=@n13
   "DDX(UVs)"=@n15
   "DDY(UVs)"=@n16
  ) {mipValueMode=TMVM_Derivative, samplerSource=SSM_Wrap_WorldGroupSettings}
  n27 = StaticSwitch(True=AppendVector(A=@n26.RGB, B=@n26.RGB), False=@n26.RGB, Value=@n18)
  n28 = MaterialFunctionCall(In=Abs(None=ComponentMask(None=@n21) {b=true}), Contrast=@n23) {materialFunction=&CheapContrast}
  FunctionOutput(None=LinearInterpolate(A=@n24, B=@n27, Alpha=ComponentMask(None=@n28.Result) {r=true})) {outputName=XYZ Texture, sortPriority=1}
  FunctionOutput(None=@n24) {outputName=XY Texture, sortPriority=1}
  FunctionOutput(None=@n27) {outputName=Z Texture, sortPriority=1}
 ]
}
