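# Material function for the Bent Baton pickaxe that layers an animated, glassy energy look over incoming material attributes in texture-masked areas.
# Two bump-offset projections of a wisp texture pan along separate head and handle directions on pre-skinned positions, faking depth.
# This noise and per-part stepped Fresnels index a color ramp, adding edge and sky-light highlights for emissive.
# Masks also drive surface overrides, Fresnel opacity and subsurface tint; several leftover branches feed no output.
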MaterialFunction /BRCosmetics/Weapons/FORT_Melee/Pickaxe_BentBaton/Materials/MF_Pickaxe_BentBaton
graph {nodes [
  n1 = TextureSample() {texture=/BRCosmetics/Weapons/FORT_Melee/Pickaxe_BentBaton/Skins/Textures/T_BentBaton_Pickaxe_Flipaway_FX2}
  n2 = Add(A=@n1.R, B=@n1.R)
  n3 = TextureSample() {samplerType=SAMPLERTYPE_Masks, texture=/BRCosmetics/Weapons/FORT_Melee/Pickaxe_BentBaton/Skins/Textures/T_BentBaton_Pickaxe_Flipaway_FX}
  Add(A=@n2, B=NamedRerouteDeclaration(None=@n3.B) {name=Handle Mask, nodeColor=<rgba 0 1 0.8359 1>})
  CameraPositionWS()
  n4 = GetMaterialAttributes(None=FunctionInput() {inputName=Result, inputType=FunctionInput_MaterialAttributes}) {attributeGetTypes [
    #1
    #2
    #3
    #4
    #5
   ]}
  n5 = Saturate(None=NamedRerouteUsage())
  n6 = LinearInterpolate(
   A=ScalarParameter() {defaultValue=-0.4176, parameterName=Second Fresnel - Step Min}
   B=ScalarParameter() {defaultValue=-0.502, parameterName=Second Fresnel - Step Max}
   Alpha=@n5
  )
  n7 = LinearInterpolate(
   A=ScalarParameter() {defaultValue=3, parameterName=Second Fresnel - Tightness Min}
   B=ScalarParameter() {defaultValue=3, parameterName=Second Fresnel - Tightness Max}
   Alpha=@n5
  )
  n8 = OneMinus(None=MaterialFunctionCall(Exponent=ScalarParameter() {defaultValue=1, parameterName=Fresnel Power}) {materialFunction=&AxisAlignedFresnel}.Fresnel)
  n9 = Saturate(None=SmoothStep(Min=@n6, Max=Add(A=@n6, B=@n7), Value=@n8))
  n10 = CameraVectorWS()
  n11 = ComponentMask(None=Add(A=Transform(None=@n10) {transformSourceType=TRANSFORMSOURCE_World, transformType=TRANSFORM_Local}) {constB=-0.1}) {r=true}
  n12 = Max(A=@n11, B=Subtract(A=OneMinus(None=@n11)) {constB=1.1})
  n13 = LinearInterpolate(
   A=ScalarParameter() {defaultValue=0.1448, parameterName=Head Fresnel - Step Min}
   B=ScalarParameter() {defaultValue=-0.4101, parameterName=Head Fresnel - Step Max}
   Alpha=@n12
  )
  n14 = ScalarParameter() {defaultValue=2, parameterName=Head Fresnel - Smoothness}
  n15 = ScalarParameter() {defaultValue=1, parameterName=ExponentIn}
  n16 = Saturate(None=SmoothStep(Min=@n13, Max=Add(A=@n13, B=@n14), Value=Fresnel(ExponentIn=@n15) {baseReflectFraction=0}))
  n17 = ScalarParameter() {defaultValue=-0.3886, parameterName=Fresnel Step Min}
  n18 = ScalarParameter() {defaultValue=4, parameterName=Fresnel Tightness}
  n19 = Saturate(None=SmoothStep(Min=@n17, Max=Add(A=@n17, B=@n18), Value=@n8))
  n20 = ScalarParameter() {defaultValue=1, parameterName=Edge Handle Fresnel}
  n21 = Max(A=Max(A=Multiply(A=@n9, B=@n1.R), B=Multiply(A=@n1.G, B=@n16)), B=Multiply(A=@n3.B, B=Multiply(A=@n19, B=@n20)))
  n22 = TextureObject() {texture=&/Swirly_Wisps_Vertical_Dissolve}
  n23 = ScalarParameter() {defaultValue=15, parameterName=Tiling}
  n24 = VertexInterpolator(VS=LocalPosition() {localOrigin=InstancePreSkinning}.XYZ)
  n25 = Add(
   A=Multiply(A=NamedRerouteUsage(), B=VectorParameter() {parameterName=Head Direction, defaultValue=<rgba 0 2 -5 1>}.RGB) {constB=2}
   B=@n24.PS
  )
  n26 = NamedRerouteUsage()
  n27 = Multiply(A=Time(), B=ScalarParameter() {defaultValue=0.4, parameterName=Global Speed})
  n28 = Add(
   A=@n24.PS
   B=Multiply(
    A=NamedRerouteDeclaration(None=@n27) {name=Global Time, nodeColor=<rgba 0.2422 1 0 1>}
    B=VectorParameter() {parameterName=Handle Direction, defaultValue=<rgba 0 -1 -5 1>}.RGB
   ) {constB=2}
  )
  n29 = NamedRerouteUsage()
  n30 = Add(A=Add(A=Multiply(A=@n25, B=@n26), B=Multiply(A=@n28, B=@n29)), B=NamedRerouteUsage())
  n31 = Multiply(A=ScalarParameter() {defaultValue=0.4, parameterName=Head Rotation}, B=NamedRerouteUsage())
  n32 = VertexInterpolator(VS=PreSkinnedNormal())
  n33 = ScalarParameter() {defaultValue=-0.35, parameterName=Main Bump}
  n34 = ScalarParameter() {parameterName="ProjectionTransitionContrast (S)"}
  n35 = Add(A=@n30, B=ScalarParameter() {defaultValue=4.875, parameterName=B_2}) {constB=6}
  n36 = ScalarParameter() {defaultValue=-0.1, parameterName=Energy - Bump Offset}
  n37 = Max(
   A=MaterialFunctionCall!(
    RotYAxis
    TextureObject=@n22
    TextureSize=@n23
    WorldPosition=@n30
    Export Float 4
    Heightmap=@n22
    RotZAxis
    RotXAxis=@n31
    World Space Normal=@n32.PS
    Parallax=@n33
    Height Amount=@n33
    Min Steps
    Max Steps
    Ref Plane
    Bump Amount=@n33
    ProjectionTransitionContrast=@n34
   ) {materialFunction=&/MF_WorldAlignedTexture_RotBumpOffset}.XYZ Texture
   B=Multiply(
    A=MaterialFunctionCall(
     TextureObject=@n22
     TextureSize=@n23
     WorldPosition=@n35
     RotXAxis=@n31
     World Space Normal=@n32.PS
     Parallax=@n36
     Bump Amount=@n36
     ProjectionTransitionContrast=@n34
    ) {materialFunction=&MF_WorldAlignedTexture_RotBumpOffset}.XYZ Texture
    B=ScalarParameter() {defaultValue=1, parameterName=Depth Value}
   )
  )
  n38 = Multiply(
   A=Power(Base=@n37, Exp=ScalarParameter() {defaultValue=0.5, parameterName=Noise Power})
   B=ScalarParameter() {defaultValue=0.4, parameterName=Noise Strength}
  )
  n39 = ScalarParameter() {parameterName=Step}
  n40 = SmoothStep(
   Min=@n39
   Max=Add(A=@n39, B=ScalarParameter() {defaultValue=2, parameterName=Smoothness})
   Value=Fresnel(ExponentIn=ScalarParameter() {defaultValue=2, parameterName=ExponentIn_1}) {baseReflectFraction=0}
  )
  n41 = Subtract(A=@n38, B=@n40)
  n42 = Max(A=@n21, B=@n41)
  n43 = Multiply(
   A=Power(Base=@n37, Exp=ScalarParameter() {defaultValue=3, parameterName=Exp_1}) {constExponent=1}
   B=ScalarParameter() {defaultValue=1, parameterName=B_6}
  )
  n44 = ScalarParameter() {defaultValue=1, parameterName=Edge Fresnel Removal}
  n45 = TextureSampleParameter2D(UVs=Subtract(A=Max(A=@n42, B=@n43), B=Multiply(A=@n19, B=@n44))) {
   parameterName=Color Scheme
   samplerType=SAMPLERTYPE_LinearColor
   texture=/BRCosmetics/Weapons/FORT_Melee/Pickaxe_BentBaton/Skins/Materials/CA_Pickaxe_Bentbaton_Flipaway
  }
  n46 = Fresnel(ExponentIn=ScalarParameter() {defaultValue=2.5, parameterName=ExponentIn_3}) {baseReflectFraction=0}
  n47 = ScalarParameter() {defaultValue=0.1, parameterName=Edge Highlight Intensity}
  n48 = ScalarParameter() {defaultValue=0.7449, parameterName=Highlight Step}
  n49 = ReflectionVectorWS()
  n50 = VertexNormalWS()
  n51 = SmoothStep(
   Min=@n48
   Max=Add(A=@n48, B=ScalarParameter() {defaultValue=1, parameterName=Highlight Smoothness})
   Value=Power(Base=DotProduct(A=@n49, B=@n50), Exp=ScalarParameter() {defaultValue=1, parameterName=Highlight Exp})
  )
  n52 = SkyLightEnvMapSample(Direction=@n49)
  n53 = Add(
   A=Add(A=@n45.RGB, B=Multiply(A=@n46, B=@n47))
   B=Multiply(A=Multiply(A=@n51, B=@n52), B=ScalarParameter() {defaultValue=1, parameterName=Fake Highlight Strength})
  )
  n54 = ScalarParameter() {defaultValue=1, parameterName=Emissive Intensity}
  n55 = Max(A=Max(A=@n3.B, B=@n1.R), B=@n1.G)
  n56 = Fresnel(BaseReflectFractionIn=ScalarParameter() {defaultValue=0.1, parameterName=BaseOpacity})
  n57 = Multiply(A=@n55, B=@n56)
  n58 = VectorParameter() {parameterName=SSS Color, defaultValue=<rgba 4 0.1254 3.6787 1>}
  n59 = Fresnel(ExponentIn=ScalarParameter() {defaultValue=0.75, parameterName=expn})
  n60 = Multiply(
   A=Power(Base=OneMinus(None=@n59), Exp=ScalarParameter() {defaultValue=2, parameterName=p2})
   B=Fresnel(ExponentIn=ScalarParameter() {defaultValue=0.75, parameterName=TOP DOWN EXP})
  )
  n61 = SetMaterialAttributes!(
   MaterialAttributes=@n4.MaterialAttributes
   EmissiveColor=LinearInterpolate(A=@n4.EmissiveColor, B=Multiply(A=@n53, B=@n54), Alpha=@n55)
   BaseColor=LinearInterpolate(A=@n4.BaseColor, Alpha=@n55) {constB=0}
   Specular=LinearInterpolate(A=@n4.Specular, B=ScalarParameter() {defaultValue=2, parameterName=Spec Bent}, Alpha=@n55)
   Roughness=LinearInterpolate(A=@n4.Roughness, B=ScalarParameter() {defaultValue=0.3, parameterName=Roughness Bent}, Alpha=@n55)
   Opacity=LinearInterpolate(B=@n57, Alpha=NamedRerouteUsage()) {constA=1}
   SubsurfaceColor=LinearInterpolate(B=Multiply(A=@n58.RGB, B=@n60), Alpha=@n57)
   Metallic=LinearInterpolate(A=@n4.Metallic, B=ScalarParameter() {parameterName=Metallic}, Alpha=@n55)
  ) {attributeSetTypes [
    #1
    #2
    #3
    #4
    #6
    #7
    #5
   ]}
  FunctionOutput(None=@n61) {outputName=SetMaterialAttributes}
  n62 = ScalarParameter() {defaultValue=-0.2, parameterName=Min}
  n63 = ScalarParameter() {defaultValue=1}
  n64 = ScalarParameter() {defaultValue=0.2, parameterName=ExponentIn_2}
  n65 = OneMinus(None=SmoothStep(Min=@n62, Max=Add(A=@n62, B=@n63), Value=Fresnel(ExponentIn=@n64) {baseReflectFraction=0}))
  LinearInterpolate(
   A=@n45.RGB
   B=Multiply(A=@n45.RGB, B=VectorParameter() {parameterName=Darkening, defaultValue=<rgba 50 0 13.6646 1>}.RGB)
   Alpha=Multiply(A=Saturate(None=@n65), B=ScalarParameter() {parameterName=Glow Opacity})
  )
  LinearInterpolate(A=@n17, B=ScalarParameter() {defaultValue=-0.0567, parameterName=Fresnel Step Max}, Alpha=NamedRerouteUsage())
  MaterialFunctionCall!(TextureObject, TextureSize, WorldPosition, Export Float 4, Parallax, In, World Space Normal, ProjectionTransitionContrast) {materialFunction=/BRCosmetics/Characters/Player/Male/Medium/Bodies/M_MED_SilverBellMarine/Materials/MF_Triplanar_ScalePan_Parallax}
  n66 = ComponentMask(None=MaterialFunctionCall!(FlowID, FlowGradient=@n1.B, FlowMask=@n1.B) {materialFunction=/Game/Global/Materials/MaterialGym/Materials/Flow/MaterialFunctions/MF_Flow}.Result) {r=true}
  Max(A=@n66, B=@n21)
  n67 = Transform(None=@n10) {transformSourceType=TRANSFORMSOURCE_World, transformType=TRANSFORM_Local}
  n68 = ComponentMask(None=RotateAboutAxis(
    NormalizedRotationAxis=Constant3Vector()
    RotationAngle=ScalarParameter() {defaultValue=1, parameterName=Angle}
    PivotPoint=ObjectPositionWS()
    Position=@n67
   )) {g=true}
  n69 = OneMinus(None=@n68)
  Max(A=@n69, B=@n68)
  n70 = ScalarParameter() {parameterName=Additional Edges - Step}
  n71 = SmoothStep(Min=@n70, Max=Add(A=@n70, B=ScalarParameter() {defaultValue=2, parameterName=Additional Edges - Smoothness}), Value=@n3.G)
  Max(A=@n71, B=@n66)
  n72 = ComponentMask(None=MaterialFunctionCall!(VectorToRotate=@n67, LookAtVector=LocalPosition().XYZ, RestingVector) {materialFunction=/Engine/Functions/Engine_MaterialFunctions01/Vectors/RotateVector}.Result) {r=true}
  n73 = Max(
   A=Add(A=@n72, B=ScalarParameter() {defaultValue=-0.4, parameterName=Camera Vec - Distance}) {constB=-0.5}
   B=Add(A=OneMinus(None=@n72), B=ScalarParameter() {defaultValue=-1.4, parameterName=B_1})
  )
  n74 = ScalarParameter() {defaultValue=5.5, parameterName=B}
  Max(A=@n73, B=Multiply(A=Add(A=@n69, B=@n74) {constB=0.8}))
  Multiply(A=@n41, B=@n56)
  Multiply(A=@n38, B=@n40)
  Multiply(A=@n71) {constB=0}
  Multiply(A=@n9, B=NamedRerouteUsage())
  NamedRerouteDeclaration(None=@n12) {name=Camera Vector Mask, nodeColor=<rgba 0.7109 1 0 1>}
  NamedRerouteDeclaration(None=@n73) {name=Neck Camera Vector, nodeColor=<rgba 0.9688 0 1 1>}
  NamedRerouteDeclaration(None=@n1.B) {name=Flow, nodeColor=<rgba 1 0.4453 0 1>}
  NamedRerouteDeclaration(None=@n2) {name=Rotated Parts, nodeColor=<rgba 1 0.2812 0 1>}
  NamedRerouteDeclaration(None=@n55) {name=Emissive Areas, nodeColor=<rgba 0.0547 1 0 1>}
  OneMinus(None=@n56)
  Subtract(A=@n56, B=@n41)
 ]}
