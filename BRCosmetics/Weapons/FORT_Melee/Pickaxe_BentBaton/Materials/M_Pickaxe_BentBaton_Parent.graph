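# Parent material for the Bent Baton pickaxe, using subsurface shading and material attributes on skeletal meshes with morph target and mesh deformer support.
# It wraps the pickaxe-specific MF_Pickaxe_BentBaton function between the shared 2023 character parent PreFX and PostFX functions.
# PreFX output feeds the pickaxe function, whose attributes go into PostFX alongside auxiliary inputs taken directly from PreFX.
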Material /BRCosmetics/Weapons/FORT_Melee/Pickaxe_BentBaton/Materials/M_Pickaxe_BentBaton_Parent
graph {
 properties {
  bHasPixelAnimation=true
  bUseMaterialAttributes=true
  bUsedWithMeshDeformer=true
  bUsedWithMorphTargets=true
  bUsedWithSkeletalMesh=true
  pixelDepthOffsetMode=PDOM_Legacy
  shadingModel=MSM_Subsurface
  usedShadingModels=Subsurface
 }
 nodes [
  n1 = MaterialFunctionCall() {materialFunction=&MF_CharacterParent_2023_PreFX}
  MaterialFunctionCall(
   Material=MaterialFunctionCall!(Result=@n1.Result) {materialFunction=&/MF_Pickaxe_BentBaton}.SetMaterialAttributes
   M=@n1.M
   Pre Skinned Local Position=@n1.M
   Distance Blend=@n1.M
   Fresnel=@n1.M
  ) {materialFunction=&MF_CharacterParent_2023_PostFX}
 ]
}
